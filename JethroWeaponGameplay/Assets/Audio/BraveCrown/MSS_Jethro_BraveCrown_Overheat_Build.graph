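# One-shot mono MetaSound source for the Jethro BraveCrown weapon's overheat build-up foley.
# Passes three wave variations, volume and random pitch spread into the shared MSS_Weapon_Reload_Start graph, reused here for overheat.
# A MapRange node converts the Pitch Shift input into that graph's base pitch, leaving the exposed Random.PitchShift.Base input unconnected.
# The resulting audio passes through a short fixed delay before output, while OnFinished comes straight from the inner graph.
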MetaSoundSource /JethroWeaponGameplay/Assets/Audio/BraveCrown/MSS_Jethro_BraveCrown_Overheat_Build
inputs [
 MasterVolume {type=Float}
 Pitch Shift {type=Float}
 Random.PitchShift.Base {type=Float}
 Random.PitchShift.Spread {type=Float}
 Sounds {type=WaveAsset:Array}
 UE.Source.OnPlay {type=Trigger}
]
interfaces [UE.OutputFormat.Mono:1.0, UE.Source.OneShot:1.0, UE.Source:1.1]
outputs [
 "UE.OutputFormat.Mono.Audio:0" {type=Audio}
 UE.Source.OneShot.OnFinished {type=Trigger}
]
pages [Default {
  input_defaults {
   MasterVolume=1.3
   Pitch Shift=0
   Random.PitchShift.Base=0
   Random.PitchShift.Spread=0.256
   UE.Source.OnPlay=null
   Sounds [
    /JethroWeaponGameplay/Assets/Audio/BraveCrown/wavs/Assault_Jethro_BraveCrown_Overheat_Build_01
    /JethroWeaponGameplay/Assets/Audio/BraveCrown/wavs/Assault_Jethro_BraveCrown_Overheat_Build_02
    /JethroWeaponGameplay/Assets/Audio/BraveCrown/wavs/Assault_Jethro_BraveCrown_Overheat_Build_03
   ]
  }
  nodes [
   MSS_Weapon_Reload_Start_0 = /FortMetasound/Weapons/MSS_Weapon_Reload_Start(
    MasterVolume=^MasterVolume
    Random.PitchShift.Spread=^Random.PitchShift.Spread
    Sounds=^Sounds
    UE.Source.OnPlay=^UE.Source.OnPlay
    Random.PitchShift.Base=MapRange.MapRange:Float(In=^Pitch Shift) {also {defaults {In Range B=0, Out Range A=1, Out Range B=12}}}.Out Value
   )
   Delay_0 = UE.Delay:Audio(In=@MSS_Weapon_Reload_Start_0."UE.OutputFormat.Mono.Audio:0") {also {defaults {Delay Time=0.1}}}
  ]
  links [
   {dst="@UE.OutputFormat.Mono.Audio:0", src=Delay_0.Out}
   {dst="@UE.Source.OneShot.OnFinished", src=MSS_Weapon_Reload_Start_0.UE.Source.OneShot.OnFinished}
  ]
 }]
properties {
 assetImportData=&self:AssetImportData
 attenuationSettings=&AS_Fort_WeaponFoley
 maxDistance=2250
 sampleRate=48000
 soundClassObject=&Fort_WeaponFoley
 concurrencySet [&CS_Fort_WeaponFoley]
}
